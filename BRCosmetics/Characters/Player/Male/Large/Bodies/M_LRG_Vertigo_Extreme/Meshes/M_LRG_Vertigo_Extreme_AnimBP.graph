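# Anim Blueprint for the M_LRG_Vertigo_Extreme body that reshapes the arms on a pose copied from the parent mesh.
# The base Vertigo Control Rig feeds always-on bone constraints and modifiers that shrink the right hand and forearm and offset upper arms and triceps.
# Events compute vertigo_pickaxe from visible weapon and tag, and set HangGliding.
# The vertigo_pickaxe-weighted bone chain is disconnected from the output, and HangGliding is never read.
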AnimBlueprint /BRCosmetics/Characters/Player/Male/Large/Bodies/M_LRG_Vertigo_Extreme/Meshes/M_LRG_Vertigo_Extreme_AnimBP
target_skeleton=&/M_LRG_Vertigo_Extreme_Skeleton
graphs [
 AnimGraph {nodes [
   n1 = Get vertigo_pickaxe() -> (vertigo_pickaxe:double)
   n2 = "Constraint - upperarm_twist_02_r (P)"(Alpha=@n1, ConstraintWeights_0=1) {
    alphaBoolBlend {blendOption=Linear}
    boneToModify {boneName=upperarm_twist_02_r}
    constraintSetup [{targetBone {boneName=lowerarm_twist_02_r}}]
    constraintWeights [1]
   }
   n3 = "Constraint - lowerarm_twist_01_r (P)"(ComponentPose=@n2, Alpha=@n1, ConstraintWeights_0=1) {
    alphaBoolBlend {blendOption=Linear}
    boneToModify {boneName=lowerarm_twist_01_r}
    constraintSetup [{targetBone {boneName=lowerarm_r}}]
    constraintWeights [1]
   }
   n4 = "Transform (Modify) Bone - Bone: hand_r"(ComponentPose=@n3, Alpha=@n1, Translation="0,0,0", Rotation="0,0,0", Scale="0.1,0.1,0.1") {scaleMode=BMM_Replace, scaleSpace=BCS_BoneSpace, alphaBoolBlend {blendOption=Linear}, boneToModify {boneName=hand_r}, scale=<0.1 0.1 0.1>}
   n5 = "Transform (Modify) Bone - Bone: lowerarm_twist_02_r"(ComponentPose=@n4, Alpha=@n1, Translation="0,0,0", Rotation="0,0,0", Scale="0.75,0.75,0.75") {
    scaleMode=BMM_Replace
    scaleSpace=BCS_BoneSpace
    alphaBoolBlend {blendOption=Linear}
    boneToModify {boneName=lowerarm_twist_02_r}
    scale=<0.75 0.75 0.75>
   }
   "Transform (Modify) Bone - Bone: lowerarm_twist_01_r"(ComponentPose=@n5, Alpha=@n1, Translation="0,0,0", Rotation="0,0,0", Scale="0.75,0.75,0.75") {
    scaleMode=BMM_Replace
    scaleSpace=BCS_BoneSpace
    alphaBoolBlend {blendOption=Linear}
    alphaScaleBiasClamp {bInterpResult=true}
    boneToModify {boneName=lowerarm_twist_01_r}
    scale=<0.75 0.75 0.75>
   }
   n6 = Control Rig(Source=Copy Pose From Mesh() {bUseAttachedParent=true}) {
    alphaBoolBlend {blendOption=Linear}
    controlRigAssetReference {blueprintRigClass=/Game/Characters/Player/Male/Large/Bodies/M_LRG_Vertigo/Meshes/M_LRG_Vertigo_CtrlRig.M_LRG_Vertigo_CtrlRig_C}
   }
   n7 = "Constraint - upperarm_twist_02_r (P)"(ComponentPose=Local To Component(LocalPose=@n6), Alpha=1, ConstraintWeights_0=1) {
    alphaBoolBlend {blendOption=Linear}
    boneToModify {boneName=upperarm_twist_02_r}
    constraintSetup [{targetBone {boneName=lowerarm_twist_02_r}}]
    constraintWeights [1]
   }
   n8 = "Constraint - lowerarm_twist_01_r (P)"(ComponentPose=@n7, Alpha=1, ConstraintWeights_0=1) {
    alphaBoolBlend {blendOption=Linear}
    boneToModify {boneName=lowerarm_twist_01_r}
    constraintSetup [{targetBone {boneName=lowerarm_r}}]
    constraintWeights [1]
   }
   n9 = "Transform (Modify) Bone - Bone: hand_r"!(ComponentPose=@n8, Alpha=1, Scale="0.1,0.1,0.1") -> (Pose) {scaleMode=BMM_Replace, scaleSpace=BCS_BoneSpace, alphaBoolBlend {blendOption=Linear}, boneToModify {boneName=hand_r}, scale=<0.1 0.1 0.1>}
   n10 = "Transform (Modify) Bone - Bone: lowerarm_twist_02_r"!(ComponentPose=@n9, Alpha=1, Scale="0.75,0.75,0.75") -> (Pose) {
    scaleMode=BMM_Replace
    scaleSpace=BCS_BoneSpace
    alphaBoolBlend {blendOption=Linear}
    boneToModify {boneName=lowerarm_twist_02_r}
    scale=<0.75 0.75 0.75>
   }
   n11 = "Transform (Modify) Bone - Bone: lowerarm_twist_01_r"!(ComponentPose=@n10, Alpha=1, Scale="0.75,0.75,0.75") -> (Pose) {
    scaleMode=BMM_Replace
    scaleSpace=BCS_BoneSpace
    alphaBoolBlend {blendOption=Linear}
    alphaScaleBiasClamp {bInterpResult=true}
    boneToModify {boneName=lowerarm_twist_01_r}
    scale=<0.75 0.75 0.75>
   }
   n12 = "Transform (Modify) Bone - Bone: upperarm_twist_01_r"!(ComponentPose=@n11, Alpha=1, Rotation="-5,0,0") -> (Pose) {
    rotationMode=BMM_Additive
    alphaBoolBlend {blendOption=Linear}
    alphaScaleBiasClamp {bInterpResult=true, interpSpeedDecreasing=5, interpSpeedIncreasing=5}
    boneToModify {boneName=upperarm_twist_01_r}
    rotation {pitch=-5}
   }
   n13 = "Transform (Modify) Bone - Bone: upperarm_twist_01_l"!(ComponentPose=@n12, Alpha=1, Rotation="5,0,0") -> (Pose) {
    rotationMode=BMM_Additive
    alphaBoolBlend {blendOption=Linear}
    alphaScaleBiasClamp {bInterpResult=true, interpSpeedDecreasing=5, interpSpeedIncreasing=5}
    boneToModify {boneName=upperarm_twist_01_l}
    rotation {pitch=5}
   }
   n14 = "Transform (Modify) Bone - Bone: tricep_r"!(ComponentPose=@n13, Alpha=1, Translation="0.3,-5,0", Rotation="-9,3,-4") -> (Pose) {
    rotationMode=BMM_Additive
    translationMode=BMM_Additive
    translationSpace=BCS_BoneSpace
    alphaBoolBlend {blendOption=Linear}
    alphaScaleBiasClamp {bInterpResult=true, interpSpeedDecreasing=5, interpSpeedIncreasing=5}
    boneToModify {boneName=tricep_r}
    rotation=<rot -9 3 -4>
    translation=<0.3 -5>
   }
   n15 = "Transform (Modify) Bone - Bone: tricep_l"!(ComponentPose=@n14, Alpha=1, Translation="1.15,-3,-0.65", Rotation="38,3.75,5.86") -> (Pose) {
    rotationMode=BMM_Additive
    translationMode=BMM_Additive
    translationSpace=BCS_BoneSpace
    alphaBoolBlend {blendOption=Linear}
    alphaScaleBiasClamp {bInterpResult=true, interpSpeedDecreasing=5, interpSpeedIncreasing=5}
    boneToModify {boneName=tricep_l}
    rotation=<rot 38 3.75 5.86>
    translation=<1.15 -3 -0.65>
   }
   Output Pose(Result=Component To Local(ComponentPose=@n15)) {layerGroup=None}
  ]}
 EventGraph {nodes [
   Event BlueprintUpdateAnimation()
   > Set FortPawn(FortPawn:Fort Pawn Object Reference=TryGetFortPawn()) -> (Output_Get:Fort Pawn Object Reference)
   n1 = Get FortPawn() -> (FortPawn:Fort Pawn Object Reference)
   > n2 = Is Valid(InputObject=@n1)
   n3 = Set weapon_visible(execute=@n2.Is Not Valid, weapon_visible:bool=false) -> (Output_Get:bool)
   n4 = Get weapon_visible() -> (weapon_visible:bool)
   n5 = Branch(execute=[@n9.then, @n3.then], Condition=@n4)
   n6 = WeaponDataHasMatchingGameplayTag!(
    self:Fort Weapon Object Reference=Get CurrentWeapon!(self:Fort Pawn Object Reference=Get FortPawn()) -> (CurrentWeapon:Fort Weapon Object Reference)
    TagToCheck:Gameplay Tag Structure=(TagName="Cosmetics.Pickaxe.Vertigo")
   ) -> (ReturnValue:bool)
   > Set vertigo_pickaxe(vertigo_pickaxe:double="To Float (Boolean)"(InBool=AND Boolean(A=@n4, B=@n6))) -> (Output_Get:double)
   Set vertigo_pickaxe(execute=@n5.else, vertigo_pickaxe=0)
   n7 = Get CurrentWeapon(self=@n1)
   n8 = IsWeaponHolstered!(self:Fort Pawn Object Reference=@n1) -> (ReturnValue:bool)
   n9 = Set weapon_visible(execute=@n2.Is Valid, weapon_visible=AND Boolean(A=IsValid(Object=@n7), B=NOT Boolean(A=@n8)))
   Event OnParachuteOpened()
   > Switch on EGliderType(Selection=Get PlayerGliderType() -> (PlayerGliderType=HangGlider))
   > Set HangGliding(HangGliding:double=1) -> (Output_Get:double)
   Event OnParachuteClosed()
   > Set HangGliding(HangGliding=0)
  ]}
]
variables [
 FortPawn {type=FortPawn, default=null}
 HangGliding {type=number}
 vertigo_pickaxe {type=number}
 weapon_visible {type=boolean}
]
